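# Crouch physics asset for the F_MED_GoldenValley female character body, covering the pelvis, left leg and dangling toy and glasses accessories.
# All bodies are single capsules, with only the left thigh, calf and foot modeled for the legs.
# A five-bone toy chain hangs from a free-rotating pelvis joint, with limit angles narrowing toward the tip.
# The two-bone glasses chain also attaches to the pelvis, its end joint set mostly for one swing axis.
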PhysicsAsset /BRCosmetics/Characters/Player/Female/Medium/Bodies/F_MED_GoldenValley/Meshes/F_MED_GoldenValley_Physics_crouch
bodies {
 calf_l [{shapeType=Capsule, length=28.5112, radius=7.8513, center=<-27.0709 -2.6252 -0.7621>, rotation=<rot 0 -87.0188 -89.5414>}]
 dyn_glasses [{shapeType=Capsule, length=15.8059, radius=2, shapeName=dyn_glasses_capsule, center=<6.0069 -1.4663 1.5>, rotation=<rot 0 -89.1524 -88.4503>}]
 dyn_glasses_base [{shapeType=Capsule, length=0.505, radius=0.505, shapeName=dyn_glasses_base_capsule, rotation=<rot 0 -90 0>}]
 dyn_toy_01 [{shapeType=Capsule, length=4, radius=1, shapeName=dyn_toy_01_capsule, center=<4 0 0>, rotation=<rot 0 90 0>}]
 dyn_toy_02 [{shapeType=Capsule, length=4, radius=1, shapeName=dyn_toy_02_capsule, center=<4 0 0>, rotation=<rot 0 90 0>}]
 dyn_toy_03 [{shapeType=Capsule, length=4, radius=2, shapeName=dyn_toy_03_capsule, center=<4 0 0>, rotation=<rot 0 90 0>}]
 dyn_toy_04 [{shapeType=Capsule, length=7, radius=3.5, shapeName=dyn_toy_04_capsule, center=<4 0 0>}]
 dyn_toy_base [{shapeType=Capsule, length=4, radius=2, shapeName=dyn_toy_base_capsule}]
 foot_l [{shapeType=Capsule, length=13.1546, radius=7.1396, center=<-2.1324 6.7595 -0.5251>, rotation=<rot 0 6.1865 -87.4773>}]
 pelvis [{shapeType=Capsule, length=8.3765, radius=10.2213, center=<2.9344 -1.9123 -0.3309>}]
 thigh_l [{shapeType=Capsule, length=32.5585, radius=8.2688, center=<-25.2243 -0.2291 -0.1117>, rotation=<rot -0.6958 -96.7195 -91.4926>}]
}
constraints [
 {
  bone1Name=dyn_toy_base
  bone2Name=pelvis
  swing1LimitDegrees=45
  swing1Motion=Free
  swing2LimitDegrees=45
  swing2Motion=Free
  twistLimitDegrees=45
  twistMotion=Free
 }
 {bone1Name=dyn_toy_01, bone2Name=dyn_toy_base, swing1LimitDegrees=60, swing2LimitDegrees=60, twistLimitDegrees=30}
 {bone1Name=dyn_toy_02, bone2Name=dyn_toy_01, swing1LimitDegrees=60, swing2LimitDegrees=60, twistLimitDegrees=30}
 {bone1Name=dyn_toy_03, bone2Name=dyn_toy_02, swing1LimitDegrees=30, swing2LimitDegrees=30, twistLimitDegrees=15}
 {bone1Name=dyn_toy_04, bone2Name=dyn_toy_03, swing1LimitDegrees=15, swing2LimitDegrees=15, twistLimitDegrees=15}
 {bone1Name=dyn_glasses_base, bone2Name=pelvis, swing1LimitDegrees=45, swing2LimitDegrees=45, twistLimitDegrees=45}
 {bone1Name=dyn_glasses, bone2Name=dyn_glasses_base, swing1LimitDegrees=60, swing2LimitDegrees=5, twistLimitDegrees=5}
 {bone1Name=thigh_l, bone2Name=pelvis, swing1LimitDegrees=45, swing2LimitDegrees=45, twistLimitDegrees=45}
 {bone1Name=calf_l, bone2Name=thigh_l, swing1LimitDegrees=45, swing2LimitDegrees=45, twistLimitDegrees=45}
 {bone1Name=foot_l, bone2Name=calf_l, swing1LimitDegrees=45, swing2LimitDegrees=45, twistLimitDegrees=45}
]
