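# Animation Blueprint for the female GoldenValley cosmetic body, adding secondary motion to the pose copied from its parent mesh.
# A Control Rig feeds blended standing and crouch RigidBody simulations, then spring AnimDynamics on trouser bones and a post-dynamics Control Rig.
# Each update sets movement states and gravity override, and drives material world-position-offset cloth from foot and calf transforms, with wind and drag varying by skydive and jump state.
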AnimBlueprint /BRCosmetics/Characters/Player/Female/Medium/Bodies/F_MED_GoldenValley/Meshes/F_MED_GoldenValley_AnimBP
target_skeleton=&/F_MED_GoldenValley_Skeleton
graphs [
 AnimGraph {nodes [
   n1 = RigidBody!(
    ComponentPose=Local To Component(LocalPose=Use cached pose 'pre_dyn'())
    bAlphaBoolEnabled:bool
    OverrideWorldGravity:Vector="0,0,0"
    ComponentLinearAccScale:Vector="0,0,0"
    ComponentLinearVelScale:Vector="0,0,0"
    bTransferBoneVelocities:bool=False
   ) -> (Pose) {
    alphaInputType=Bool
    bClampLinearTranslationLimitToRefPose=true
    bFreezeIncomingPoseOnStart=true
    bOverrideWorldGravity=true
    bTransferBoneVelocities=true
    overridePhysicsAsset=&/F_MED_GoldenValley_Physics_crouch
    simulationSpace=BaseBoneSpace
    alphaBoolBlend {blendOption=Linear}
    baseBoneRef {boneName=spine_02}
    componentAppliedLinearAccClamp=<100 100 100>
    componentLinearAccScale=<0.6 0.2 0.2>
    componentLinearVelScale=<0.6 0.2 0.2>
    simSpaceSettings {
     dampingAlpha=0
     maxAngularAcceleration=50
     maxAngularVelocity=8
     maxLinearAcceleration=1000
     maxLinearVelocity=500
     worldAlpha=1
     externalLinearDragV=<1 1 1>
    }
   }
   n2 = Component To Local(ComponentPose=@n1)
   n3 = RigidBody(
    ComponentPose=Local To Component(LocalPose=Use cached pose 'pre_dyn'())
    OverrideWorldGravity="0,0,0"
    ComponentLinearAccScale="0,0,0"
    ComponentLinearVelScale="0,0,0"
    bTransferBoneVelocities=False
   ) {
    alphaInputType=Bool
    bClampLinearTranslationLimitToRefPose=true
    bFreezeIncomingPoseOnStart=true
    bOverrideWorldGravity=true
    bTransferBoneVelocities=true
    overridePhysicsAsset=&/F_MED_GoldenValley_Physics
    simulationSpace=BaseBoneSpace
    alphaBoolBlend {blendOption=Linear}
    baseBoneRef {boneName=spine_02}
    componentAppliedLinearAccClamp=<100 100 100>
    componentLinearAccScale=<0.6 0.2 0.2>
    componentLinearVelScale=<0.6 0.2 0.2>
    simSpaceSettings {
     dampingAlpha=0
     maxAngularAcceleration=10
     maxAngularVelocity=4
     maxLinearAcceleration=600
     maxLinearVelocity=500
     worldAlpha=1
     externalLinearDragV=<1 1 1>
    }
   }
   n4 = Component To Local(ComponentPose=@n3)
   n5 = Anim Dynamics - Bone: dyn_trouser_leg_r(
    ComponentPose=Local To Component(LocalPose=Blend Poses by bool(BlendPose_0=@n2, BlendPose_1=@n4, BlendTime_0=0.4, BlendTime_1=0.4))
    Alpha=1
    ExternalForce="0,0,0"
   ) {
    bGravityOverrideInSimSpace=true
    bLinearSpring=true
    bOverrideLinearDamping=true
    bUseGravityOverride=true
    linearDampingOverride=0.9999
    linearSpringConstant=100
    simulationSpace=BoneRelative
    alphaBoolBlend {blendOption=Linear}
    boundBone {boneName=dyn_trouser_leg_r}
    componentAppliedLinearAccClamp=<3000 30 3000>
    componentLinearAccScale=<0.1 0.1 0.1>
    componentLinearVelScale=<0.9 0.9 0.9>
    gravityOverride {z=-980}
    physicsBodyDefinitions [{
      boundBone {boneName=dyn_trouser_leg_r}
      boxExtents=<60 60 60>
      constraintSetup {angularLimitsMax=<0 0 0>, angularLimitsMin=<0 0 0>, angularTarget=<0 0 0>, linearAxesMax=<5 4 5>, linearAxesMin=<-1.5 -4 -5>}
      localJointOffset=<0 0 0>
     }]
    relativeSpaceBone {boneName=calf_l}
    simSpaceSettings {maxAngularAcceleration=10}
   }
   n6 = Anim Dynamics - Bone: dyn_trouser_leg_l(ComponentPose=@n5, Alpha=1, ExternalForce="0,0,0") {
    bGravityOverrideInSimSpace=true
    bLinearSpring=true
    bOverrideLinearDamping=true
    bUseGravityOverride=true
    linearDampingOverride=0.9999
    linearSpringConstant=100
    simulationSpace=BoneRelative
    alphaBoolBlend {blendOption=Linear}
    boundBone {boneName=dyn_trouser_leg_l}
    componentAppliedLinearAccClamp=<3000 30 3000>
    componentLinearAccScale=<0.1 0.1 0.1>
    componentLinearVelScale=<0.9 0.9 0.9>
    gravityOverride {z=-980}
    physicsBodyDefinitions [{
      boundBone {boneName=dyn_trouser_leg_l}
      boxExtents=<60 60 60>
      constraintSetup {angularLimitsMax=<0 0 0>, angularLimitsMin=<0 0 0>, angularTarget=<0 0 0>, linearAxesMax=<5 4 5>, linearAxesMin=<-1.5 -4 -5>}
      localJointOffset=<0 0 0>
     }]
    relativeSpaceBone {boneName=calf_l}
    simSpaceSettings {maxAngularAcceleration=10}
   }
   n7 = Control Rig!(Source=Component To Local(ComponentPose=@n6), Alpha, emote:bool=false, move:bool=false, crouch:bool=false, FE:bool=false) -> (Pose) {controlRigAssetReference {blueprintRigClass=&/F_MED_GoldenValley_CtrlRig_PostDyn.F_MED_GoldenValley_CtrlRig_PostDyn_C}}
   Output Pose(Result=@n7) {layerGroup=None}
   n8 = Control Rig!(Source=Copy Pose From Mesh() {bUseAttachedParent=true}, Alpha) -> (Pose) {controlRigAssetReference {blueprintRigClass=&/F_MED_GoldenValley_CtrlRig.F_MED_GoldenValley_CtrlRig_C}}
   Save cached pose 'pre_dyn'(Pose=@n8) {cacheName=pre_dyn}
  ]}
 EventGraph {nodes [
   n1 = Event BlueprintUpdateAnimation()
   > n2 = Sequence!(execute) -> (then_0, then_1, then_2:Exec, then_3:Exec, then_4:Exec)
   n3 = Set crouch(execute=@n2.then_1, crouch:bool=Get bIsCrouching()) -> (Output_Get:bool)
   > n4 = Set dyn_crouch(dyn_crouch:bool=@n3) -> (Output_Get:bool)
   > Set dyn(dyn:bool=NOT Boolean(A=@n4)) -> (Output_Get:bool)
   > Set Is Moving 2D(Is Moving 2D:bool=Get bIsMoving2D()) -> (Output_Get:bool)
   > Set Is Playing Emote(Is Playing Emote:bool=Get bIsPlayingEmote()) -> (Output_Get:bool)
   n5 = "Gravity Override State Function (Parameter Asset) (Nativized)"(
    execute=@n2.then_0
    CharacterPartAnimInstance=Self-Reference()
    GameTimeInSeconds=GetGameTimeInSeconds()
    DebugJointName=spine_05
    Parameters=&/F_MED_GoldenValley_Parameters
   )
   > Set GravityOverride(GravityOverride=@n5)
   Set WPO Globals Function(
    execute=@n2.then_2
    Material=Get Mat Instance Dynamic()
    Overlay Material=Get Overlay Mat Instance Dynamic()
    Skeletal Mesh=Get SkelMesh()
    "Local Noise Transform  Location"="Get Local Noise Transform  Location"() -> ("Local Noise Transform  Location":Vector)
    Local Noise Offset=Get Local Noise Offset()
   )
   > Set WPO Regional Function!(
    execute
    Material:Material Instance Dynamic Object Reference=Get Mat Instance Dynamic()
    Overlay Material:Material Instance Dynamic Object Reference=Get Overlay Mat Instance Dynamic()
    Skeletal Mesh:Skeletal Mesh Component Object Reference=Get SkelMesh()
    Velocity Transform:Name=foot_r
    Velocity Interp Speed:double=Get Velocity Interp Speed()
    Collision Transform:Name=calf_r
    Collision Transform Axis:EAxis Enum
    Invert Collision Transform Axis:bool=false
    Delta Time X:double=@n1.DeltaTimeX
    Previous Velocity Transform Location:"Vector (by ref)"=Get R1 Velocity Transform Location()
    Previous Velocity Transform Velocity:"Vector (by ref)"=Get R1 Velocity Transform Velocity()
    Region:int=1
   ) -> (then)
   > Set WPO Regional Function(
    Material=Get Mat Instance Dynamic()
    Overlay Material=Get Overlay Mat Instance Dynamic()
    Skeletal Mesh=Get SkelMesh()
    Velocity Transform=foot_l
    Velocity Interp Speed=Get Velocity Interp Speed()
    Collision Transform=calf_l
    Invert Collision Transform Axis=true
    Delta Time X=@n1.DeltaTimeX
    Previous Velocity Transform Location=Get R2 Velocity Transform Location()
    Previous Velocity Transform Velocity=Get R2 Velocity Transform Velocity()
    Region=2
   )
   > n6 = Sequence!(execute) -> (then_0, then_1)
   > n7 = Branch(Condition=Get bIsSkydiveFloating())
   > n8 = Set WPO Pho Wind Intensity(WPO Pho Wind Intensity=1)
   SetScalarParameterValue(
    execute=[@n8.then, @n9.then]
    self=[Get Mat Instance Dynamic(), Get Overlay Mat Instance Dynamic()]
    ParameterName=ClothSim_WindStrength
    Value=Get WPO Pho Wind Intensity()
   )
   n9 = Set WPO Pho Wind Intensity(execute=@n7.else, WPO Pho Wind Intensity=0)
   n10 = Get bIsSkyDiving()
   n11 = Branch(execute=@n6.then_1, Condition=AND Boolean(A=Get bIsJumping(), B=NOT Boolean(A=@n10)))
   > n12 = Set WPO Pho Wind Intensity(WPO Pho Wind Intensity=0.75)
   n13 = Get Mat Instance Dynamic()
   n14 = Get Overlay Mat Instance Dynamic()
   n15 = Get WPO Pho Wind Intensity()
   SetScalarParameterValue(execute=[@n16.then, @n12.then], self=[@n13, @n14], ParameterName=ClothSim_R1_DragStrength, Value=@n15)
   > SetScalarParameterValue(self=[@n13, @n14], ParameterName=ClothSim_R2_DragStrength, Value=@n15)
   n16 = Set WPO Pho Wind Intensity(execute=@n11.else, WPO Pho Wind Intensity=1)
   Event OnCharacterCustomizationLoaded()
   > n17 = Set SkelMesh(SkelMesh=GetPartSkeletalMeshComponent())
   > n18 = Get Dynamic Material Instance(SkeletalMeshComponent=@n17)
   > Set Mat Instance Dynamic(Mat Instance Dynamic=@n18)
   > n19 = GetOrCreateDynamicOverlayMaterial(MeshComponent=@n17, MaterialName=None)
   > Set Overlay Mat Instance Dynamic(Overlay Mat Instance Dynamic=@n19)
  ]}
]
variables [
 crouch {type=boolean}
 dyn {type=boolean}
 dyn_crouch {type=boolean}
 Gravity Override Params {type=GravityOverrideParamsStruct}
 GravityOverride {type=Vector}
 Is Moving 2D {type=boolean}
 Is Playing Emote {type=boolean}
 Local Noise Offset {type=Vector}
 "Local Noise Transform  Location" {type=Vector}
 Mat Instance Dynamic {type=MaterialInstanceDynamic}
 Overlay Mat Instance Dynamic {type=MaterialInstanceDynamic}
 Parameters {type=FortCharacterDynamicsParameters, default=null}
 R1 Velocity Transform Location {type=Vector}
 R1 Velocity Transform Velocity {type=Vector}
 R2 Velocity Transform Location {type=Vector}
 R2 Velocity Transform Velocity {type=Vector}
 SkelMesh {type=SkeletalMeshComponent}
 Velocity Interp Speed {type=number}
 WPO Pho Wind Intensity {type=number}
]
