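# Post-dynamics Control Rig that corrects hair and headscarf bones on the F_MED_GoldenValley face accessory.
# A head spherical pose reader, smoothed and disabled during skydive, drives hair and scarf offsets; upper-arm readers set variables nothing uses.
# Back hair is capsule-clamped from spine_05_ctrl outside skydive, then parent-constrained to upper-body bones when Backpack or Cape is set, middle joints at half weight.
# FE without Emote adds side and front hair rotations.
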ControlRigBlueprint /BRCosmetics/Characters/Player/Female/Medium/Bodies/F_MED_GoldenValley/Meshes/Parts/F_MED_GoldenValley_FaceAcc_CtrlRig_PostDyn
graphs [
 RigVMFunctionLibrary {links [], nodes []}
 RigVMFunction_Sequence_1_ContainedGraph {nodes [
   n1 = RigVMFunctionEntryNode()
   RigVMFunction_Sequence = &RigVMFunction_Sequence(ExecuteContext=@n1.ExecuteContext)
   n2 = &RigVMFunction_Sequence(ExecuteContext=@RigVMFunction_Sequence.B)
   n3 = &RigVMFunction_Sequence(ExecuteContext=@n2.B)
   n4 = &RigVMFunction_Sequence(ExecuteContext=@n3.B)
   RigVMFunctionReturnNode(A=@RigVMFunction_Sequence.A, B=@n2.A, C=@n3.A, D=@n4.A, E=@n4.B)
  ]}
 RigVMFunction_Sequence_5_ContainedGraph {nodes [
   n1 = RigVMFunctionEntryNode()
   n2 = &RigVMFunction_Sequence(ExecuteContext=@n1.ExecuteContext)
   n3 = &RigVMFunction_Sequence(ExecuteContext=@n2.B)
   RigVMFunctionReturnNode(A=@n2.A, B=@n3.A, C=@n3.B)
  ]}
 RigVMFunction_Sequence_9_ContainedGraph {nodes [
   n1 = RigVMFunctionEntryNode()
   n2 = &RigVMFunction_Sequence(ExecuteContext=@n1.ExecuteContext)
   n3 = &RigVMFunction_Sequence(ExecuteContext=@n2.B)
   n4 = &RigVMFunction_Sequence(ExecuteContext=@n3.B)
   RigVMFunctionReturnNode(B=@n3.A, C=@n4.A, D=@n4.B, A=@n2.A)
  ]}
 RigVMModel {nodes [
   RigUnit_BeginExecution = &RigUnit_BeginExecution()
   RigVMFunction_Sequence_13 = RigVMAggregateNode(ExecuteContext=@RigUnit_BeginExecution.ExecutePin)
   RigVMFunction_Sequence_11 = RigVMAggregateNode(ExecuteContext=@RigVMFunction_Sequence_13.A)
   n1 = &RigUnit_SphericalPoseReader(
    ActiveRegionScaleFactors=(PositiveWidth=1,NegativeWidth=0.5,PositiveHeight=0.75,NegativeHeight=0.75)
    ActiveRegionSize=0.1
    Debug=(bDrawDebug=true,bDraw2D=False,bDrawLocalAxes=False,DebugScale=25,DebugSegments=20,DebugThickness=0.25)
    DriverAxis=(X=0,Y=1,Z=0)
    DriverItem=(Type=Bone,Name="head")
    FalloffRegionScaleFactors=(PositiveWidth=1,NegativeWidth=0.5,PositiveHeight=1,NegativeHeight=1)
    FalloffSize=0.3
    FlipHeightScaling=False
    FlipWidthScaling=False
    OptionalParentItem=(Type=Bone,Name="None")
    RotationOffset=(X=90,Y=0,Z=-39)
    ExecutePin=@RigVMFunction_Sequence_11.B
   )
   n2 = &RigVMFunction_AlphaInterp(InterpSpeedDecreasing=10, InterpSpeedIncreasing=10, Scale=1, bClampResult=False, Value=@n1.OutputParam)
   n3 = RigVMVariableNode() {variable=Alpha_Skydive}
   > RigVMVariableNode(Value=&RigVMFunction_MathDoubleMul(
     A=@n2.Result
     B=&RigVMFunction_MathDoubleRemap(SourceMaximum=1, SourceMinimum=0, TargetMaximum=0, TargetMinimum=1, bClamp=False, Value=@n3.Value).Result
    ).Result) {variable=SPR_head_up}
   n4 = &RigUnit_SphericalPoseReader(
    ActiveRegionScaleFactors=(PositiveWidth=1,NegativeWidth=0.5,PositiveHeight=0.75,NegativeHeight=0.75)
    ActiveRegionSize=0.2
    Debug=(bDrawDebug=true,bDraw2D=False,bDrawLocalAxes=False,DebugScale=25,DebugSegments=20,DebugThickness=0.25)
    DriverAxis=(X=-1,Y=0,Z=0)
    DriverItem=(Type=Bone,Name="upperarm_r")
    FalloffRegionScaleFactors=(PositiveWidth=1,NegativeWidth=0.5,PositiveHeight=1,NegativeHeight=1)
    FalloffSize=0.3
    FlipHeightScaling=False
    FlipWidthScaling=False
    OptionalParentItem=(Type=Bone,Name="None")
    RotationOffset=(X=134,Y=8,Z=106)
    ExecutePin=@RigVMFunction_Sequence_11.C
   )
   n5 = &RigVMFunction_AlphaInterp(InterpSpeedDecreasing=10, InterpSpeedIncreasing=10, Scale=1, bClampResult=False, Value=@n4.OutputParam)
   > RigVMVariableNode(Value=@n5.Result) {variable=Upperarm_R}
   n6 = &RigUnit_SphericalPoseReader(
    ActiveRegionScaleFactors=(PositiveWidth=1,NegativeWidth=0.5,PositiveHeight=0.75,NegativeHeight=0.75)
    ActiveRegionSize=0.2
    Debug=(bDrawDebug=true,bDraw2D=False,bDrawLocalAxes=False,DebugScale=25,DebugSegments=20,DebugThickness=0.25)
    DriverAxis=(X=1,Y=0,Z=0)
    DriverItem=(Type=Bone,Name="upperarm_l")
    FalloffRegionScaleFactors=(PositiveWidth=1,NegativeWidth=0.5,PositiveHeight=1,NegativeHeight=1)
    FalloffSize=0.3
    FlipHeightScaling=False
    FlipWidthScaling=False
    OptionalParentItem=(Type=Bone,Name="None")
    RotationOffset=(X=47,Y=-3,Z=-103)
    ExecutePin=@RigVMFunction_Sequence_11.D
   )
   n7 = &RigVMFunction_AlphaInterp(InterpSpeedDecreasing=10, InterpSpeedIncreasing=10, Scale=1, bClampResult=False, Value=@n6.OutputParam)
   > RigVMVariableNode(Value=@n7.Result) {variable=Upperarm_L}
   RigVMVariableNode(
    Value=&RigVMFunction_MathBoolToFloat(Value=RigVMVariableNode() {variable=Skydive}.Value).Result
    ExecuteContext=@RigVMFunction_Sequence_11.A
   ) {variable=Alpha_Skydive}
   > RigVMVariableNode(Value=&RigVMFunction_MathBoolToFloat(Value=RigVMVariableNode() {variable=Backpack}.Value).Result) {variable=Alpha_backpack}
   > RigVMVariableNode(Value=&RigVMFunction_MathBoolToFloat(Value=RigVMVariableNode() {variable=Cape}.Value).Result) {variable=Alpha_cape}
   n8 = &RigVMFunction_Sequence(ExecuteContext=@RigVMFunction_Sequence_13.B)
   &RigUnit_OffsetTransformForItem(
    Item=(Type=Bone,Name="dyn_hair_03_bk_r")
    OffsetTransform=(Rotation=(X=0,Y=0,Z=0,W=1),Translation=(X=0,Y=0,Z=-10),Scale3D=(X=1,Y=1,Z=1))
    bPropagateToChildren=True
    Weight=RigVMVariableNode() {variable=SPR_head_up}.Value
    ExecutePin=@n8.A
   )
   > &RigUnit_OffsetTransformForItem(
    Item=(Type=Bone,Name="dyn_hair_03_bk_c")
    OffsetTransform=(Rotation=(X=0,Y=0,Z=0,W=1),Translation=(X=0,Y=0,Z=-10),Scale3D=(X=1,Y=1,Z=1))
    bPropagateToChildren=True
    Weight=RigVMVariableNode() {variable=SPR_head_up}.Value
   )
   > &RigUnit_OffsetTransformForItem(
    Item=(Type=Bone,Name="dyn_hair_03_bk_l")
    OffsetTransform=(Rotation=(X=0,Y=0,Z=0,W=1),Translation=(X=0,Y=0,Z=-10),Scale3D=(X=1,Y=1,Z=1))
    bPropagateToChildren=True
    Weight=RigVMVariableNode() {variable=SPR_head_up}.Value
   )
   RigVMFunction_Sequence_9 = RigVMAggregateNode(ExecuteContext=@n8.B)
   n9 = &RigVMFunction_MathDoubleRemap(SourceMaximum=1, SourceMinimum=0, TargetMaximum=0, TargetMinimum=1, bClamp=False, Value=RigVMVariableNode() {variable=Alpha_Skydive}.Value)
   &RigUnit_SetTransform(
    Item=(Type=Bone,Name="dyn_hair_03_bk_c")
    bInitial=False
    Value=/Script/RigVM.RigVMFunction_MathTransformClampSpatially(
     Axis=X
     DebugColor=(R=1,G=0,B=0,A=1)
     DebugThickness=1
     Maximum=50
     Minimum=12
     Type=Capsule
     bDrawDebug=true
     Value=&RigUnit_GetTransform(Item=(Type=Bone,Name="dyn_hair_03_bk_c"), bInitial=False).Transform
     Space=&RigUnit_GetTransform(Item=(Type=Control,Name="spine_05_ctrl"), bInitial=False).Transform
    ) {
     is=struct RigVMUnitNode
     template="ClampSpatially::Execute(in Value,in Axis,in Type,in Minimum,in Maximum,in Space,in bDrawDebug,in DebugColor,in DebugThickness,out Result)"
    }.Result
    ExecutePin=@RigVMFunction_Sequence_9.A
    Weight=@n9.Result
   )
   &RigUnit_SetTransform(
    Item=(Type=Bone,Name="dyn_hair_03_bk_l")
    bInitial=False
    Value=&RigVMFunction_MathTransformClampSpatially(
     Axis=X
     DebugColor=(R=1,G=0,B=0,A=1)
     DebugThickness=1
     Maximum=50
     Minimum=15
     Type=Capsule
     bDrawDebug=true
     Value=&RigUnit_GetTransform(Item=(Type=Bone,Name="dyn_hair_03_bk_l"), bInitial=False).Transform
     Space=&RigUnit_GetTransform(Item=(Type=Control,Name="spine_05_ctrl"), bInitial=False).Transform
    ) {template="ClampSpatially::Execute(in Value,in Axis,in Type,in Minimum,in Maximum,in Space,in bDrawDebug,in DebugColor,in DebugThickness,out Result)"}.Result
    ExecutePin=@RigVMFunction_Sequence_9.B
    Weight=@n9.Result
   )
   &RigUnit_SetTransform(
    Item=(Type=Bone,Name="dyn_hair_03_bk_r")
    bInitial=False
    Value=&RigVMFunction_MathTransformClampSpatially(
     Axis=X
     DebugColor=(R=1,G=0,B=0,A=1)
     DebugThickness=1
     Maximum=50
     Minimum=15
     Type=Capsule
     bDrawDebug=true
     Value=&RigUnit_GetTransform(Item=(Type=Bone,Name="dyn_hair_03_bk_r"), bInitial=False).Transform
     Space=&RigUnit_GetTransform(Item=(Type=Control,Name="spine_05_ctrl"), bInitial=False).Transform
    ) {template="ClampSpatially::Execute(in Value,in Axis,in Type,in Minimum,in Maximum,in Space,in bDrawDebug,in DebugColor,in DebugThickness,out Result)"}.Result
    ExecutePin=@RigVMFunction_Sequence_9.C
    Weight=@n9.Result
   )
   n10 = &RigVMFunction_Sequence(ExecuteContext=@RigVMFunction_Sequence_13.C)
   Maximum = &RigVMFunction_MathDoubleMax(B=RigVMVariableNode() {variable=Alpha_cape}.Value, A=RigVMVariableNode() {variable=Alpha_backpack}.Value)
   n11 = &RigVMFunction_MathDoubleMul(B=0.5, A=@Maximum.Result)
   &RigUnit_ParentConstraint(
    AdvancedSettings=(RotationOrderForFilter=XZY)
    Child=(Type=Bone,Name="dyn_hair_02_bk_l")
    Parents=((Item=(Type=Bone,Name="spine_05"),Weight=0.4341),(Item=(Type=Bone,Name="clavicle_l"),Weight=0.5659))
    ExecutePin=@n10.A
    Weight=@n11.Result
   )
   > &RigUnit_ParentConstraint(
    AdvancedSettings=(RotationOrderForFilter=XZY)
    Child=(Type=Bone,Name="dyn_hair_02_bk_c")
    Parents=((Item=(Type=Bone,Name="neck_01"),Weight=0.0068),(Item=(Type=Bone,Name="spine_04"),Weight=0.0091),(Item=(Type=Bone,Name="clavicle_l"),Weight=0.0187),(Item=(Type=Bone,Name="clavicle_r"),Weight=0.1953),(Item=(Type=Bone,Name="spine_05"),Weight=0.7701))
    Weight=@n11.Result
   )
   > &RigUnit_ParentConstraint(
    AdvancedSettings=(RotationOrderForFilter=XZY)
    Child=(Type=Bone,Name="dyn_hair_02_bk_r")
    Parents=((Item=(Type=Bone,Name="spine_05"),Weight=0.4341),(Item=(Type=Bone,Name="clavicle_r"),Weight=0.5659))
    Weight=@n11.Result
   )
   &RigUnit_ParentConstraint(
    AdvancedSettings=(RotationOrderForFilter=XZY)
    Child=(Type=Bone,Name="dyn_hair_03_bk_l")
    Parents="((Item=(Type=Bone,Name=\"spine_04\"),Weight=0.0027),(Item=(Type=Bone,Name=\"deform_upperarm_bk_l\"),Weight=0.004),(Item=(Type=Bone,Name=\"clavicle_r\"),Weight=0.0047),(Item=(Type=Bone,Name=\"deform_upperarm_up_l\"),Weight=0.0105),(Item=(Type=Bone,Name=\"upperarm_tw ...(truncated)"
    ExecutePin=@n10.B
    Weight=@Maximum.Result
   )
   > &RigUnit_ParentConstraint(
    AdvancedSettings=(RotationOrderForFilter=XZY)
    Child=(Type=Bone,Name="dyn_hair_03_bk_c")
    Parents=((Item=(Type=Bone,Name="neck_01"),Weight=0.0068),(Item=(Type=Bone,Name="spine_04"),Weight=0.0091),(Item=(Type=Bone,Name="clavicle_l"),Weight=0.0187),(Item=(Type=Bone,Name="clavicle_r"),Weight=0.1953),(Item=(Type=Bone,Name="spine_05"),Weight=0.7701))
    Weight=@Maximum.Result
   )
   > &RigUnit_ParentConstraint(
    AdvancedSettings=(RotationOrderForFilter=XZY)
    Child=(Type=Bone,Name="dyn_hair_03_bk_r")
    Parents="((Item=(Type=Bone,Name=\"spine_04\"),Weight=0.0027),(Item=(Type=Bone,Name=\"deform_upperarm_bk_r\"),Weight=0.004),(Item=(Type=Bone,Name=\"clavicle_l\"),Weight=0.0047),(Item=(Type=Bone,Name=\"deform_upperarm_up_r\"),Weight=0.0105),(Item=(Type=Bone,Name=\"upperarm_tw ...(truncated)"
    Weight=@Maximum.Result
   )
   &RigUnit_OffsetTransformForItem(
    Item=(Type=Bone,Name="dyn_headscarf_base")
    OffsetTransform=(Rotation=(X=0,Y=0,Z=0,W=1),Translation=(X=-3,Y=0,Z=1),Scale3D=(X=1,Y=1,Z=1))
    bPropagateToChildren=true
    Weight=RigVMVariableNode() {variable=SPR_head_up}.Value
    ExecutePin=@RigVMFunction_Sequence_13.D
   )
   Remap = &RigVMFunction_MathDoubleRemap(SourceMaximum=1, SourceMinimum=0, TargetMaximum=0, TargetMinimum=1, bClamp=False, Value=RigVMVariableNode() {variable=SPR_head_up}.Value)
   > &RigUnit_OffsetTransformForItem(
    Item=(Type=Bone,Name="dyn_hair_03_side_l")
    OffsetTransform=(Rotation=(X=0,Y=0,Z=0,W=1),Translation=(X=3,Y=0,Z=0),Scale3D=(X=1,Y=1,Z=1))
    bPropagateToChildren=True
    Weight=@Remap.Result
   )
   RigVMFunction_MathBoolNot = &RigVMFunction_MathBoolNot(Value=RigVMVariableNode() {variable=Emote}.Value)
   RigVMFunction_MathBoolAnd = &RigVMFunction_MathBoolAnd(B=@RigVMFunction_MathBoolNot.Result, A=RigVMVariableNode() {variable=FE}.Value)
   RigVMFunction_MathBoolToFloat = &RigVMFunction_MathBoolToFloat(Value=@RigVMFunction_MathBoolAnd.Result)
   > &RigUnit_OffsetTransformForItem(
    Item=(Type=Bone,Name="dyn_hair_03_side_r")
    OffsetTransform=(Rotation=(X=-0.113,Y=0.4957,Z=0.0653,W=0.8586),Translation=(X=0,Y=0,Z=0),Scale3D=(X=1,Y=1,Z=1))
    bPropagateToChildren=true
    Weight=@RigVMFunction_MathBoolToFloat.Result
   )
   > &RigUnit_OffsetTransformForItem(
    Item=(Type=Bone,Name="dyn_hair_01_fr_l")
    OffsetTransform=(Rotation=(X=0,Y=-0.0872,Z=0,W=0.9962),Translation=(X=0,Y=0,Z=0),Scale3D=(X=1,Y=1,Z=1))
    bPropagateToChildren=False
    Weight=@RigVMFunction_MathBoolToFloat.Result
   )
   > &RigUnit_OffsetTransformForItem(
    Item=(Type=Bone,Name="dyn_hair_02_fr_l")
    OffsetTransform=(Rotation=(X=0,Y=-0.1305,Z=0,W=0.9914),Translation=(X=0,Y=0,Z=0),Scale3D=(X=1,Y=1,Z=1))
    bPropagateToChildren=False
    Weight=@RigVMFunction_MathBoolToFloat.Result
   )
   > &RigUnit_OffsetTransformForItem(
    Item=(Type=Bone,Name="dyn_hair_03_fr_l")
    OffsetTransform=(Rotation=(X=0,Y=0,Z=0.0872,W=0.9962),Translation=(X=0,Y=0,Z=0),Scale3D=(X=1,Y=1,Z=1))
    bPropagateToChildren=False
    Weight=@RigVMFunction_MathBoolToFloat.Result
   )
  ]}
]
hierarchy [
 root {loc=<0 0 0>, rot=<0 0 0 1>}
 pelvis {parent=root, loc=<0 -0.5301 98.0571>, rot=<0.0127 0.707 -0.0127 -0.707>}
 spine_01 {parent=pelvis, loc=<6.1208 -0.9328 0>, rot=<0 0 0.034 -0.9994>}
 spine_02 {parent=spine_01, loc=<6.5997 -0.2095 0>, rot=<0 0 0.0238 -0.9997>}
 spine_03 {parent=spine_02, loc=<6.4914 -0.1237 0>, rot=<0 0 0.0149 -0.9999>}
 spine_04 {parent=spine_03, loc=<6.9598 0.0794 0>, rot=<0 0 -0.0132 -0.9999>}
 spine_05 {parent=spine_04, loc=<6.9987 0 0>, rot=<0 0 -0.0015 -1>}
 neck_01 {parent=spine_05, loc=<11.3138 0.1529 0>, rot=<0 0 -0.2003 -0.9797>}
 neck_02 {parent=neck_01, loc=<4.5627 0 0>, rot=<0 0 0.0504 -0.9987>}
 head {parent=neck_02, loc=<5.3914 0 0>, rot=<0 0 0.115 -0.9934>}
 dyn_headscarf_base {parent=head, loc=<10.3043 -4.9825 -0.0893>, rot=<0.1878 -0.6831 0.6803 -0.1878>}
 dyn_headscarf {parent=dyn_headscarf_base, loc=<0.1 0 0>, rot=<0 0 0 1>}
 dyn_sun_plushie_base {parent=head, loc=<10.2216 5.2627 -4.2788>, rot=<-0.3717 -0.431 -0.1684 -0.8048>}
 dyn_sun_plushie {parent=dyn_sun_plushie_base, loc=<0.1 0 0>, rot=<0 0 0 1>}
 dyn_hair_fr_down_r_base {parent=head, loc=<-3.7514 -1.0436 7.4125>, rot=<-0.0897 -0.0362 0.971 0.2186>}
 dyn_hair_01_fr_down_r {parent=dyn_hair_fr_down_r_base, loc=<0.2753 0 0>, rot=<0 0 0 1>}
 dyn_hair_02_fr_down_r {parent=dyn_hair_01_fr_down_r, loc=<5.4391 0 0>, rot=<0.0119 0.1273 0.092 -0.9875>}
 dyn_hair_03_fr_down_r {parent=dyn_hair_02_fr_down_r, loc=<5.4391 0 0>, rot=<0 0 -0.0206 -0.9998>}
 dyn_hair_fr_l_base {parent=head, loc=<5.1137 3.6709 -6.8462>, rot=<-0.0194 -0.9996 -0.0167 0.0149>}
 dyn_hair_01_fr_l {parent=dyn_hair_fr_l_base, loc=<0.2995 0 0>, rot=<0 0 0 1>}
 dyn_hair_02_fr_l {parent=dyn_hair_01_fr_l, loc=<8.2386 0 0>, rot=<0.0042 -0.0559 -0.0757 -0.9956>}
 dyn_hair_03_fr_l {parent=dyn_hair_02_fr_l, loc=<8.2386 0 0>, rot=<0 0 -0.0575 -0.9983>}
 dyn_hair_side_l_base {parent=head, loc=<3.8999 -2.0481 -5.8012>, rot=<0.0183 -0.9864 0.0148 -0.1625>}
 dyn_hair_01_side_l {parent=dyn_hair_side_l_base, loc=<0.2995 0 0>, rot=<0 0 0 1>}
 dyn_hair_02_side_l {parent=dyn_hair_01_side_l, loc=<5.4391 0 0>, rot=<-0.0635 0.0352 -0.1198 -0.9901>}
 dyn_hair_03_side_l {parent=dyn_hair_02_side_l, loc=<5.4391 0 0>, rot=<0.2472 -0.1281 -0.0002 -0.9604>}
 dyn_hair_bk_r_base {parent=head, loc=<2.9686 -7.3459 5.3361>, rot=<0.1456 -0.6934 0.7054 -0.0222>}
 dyn_hair_01_bk_r {parent=dyn_hair_bk_r_base, loc=<0.2995 0 0>, rot=<0 0 0 1>}
 dyn_hair_02_bk_r {parent=dyn_hair_01_bk_r, loc=<6.4619 0 0>, rot=<0 0 0 1>}
 dyn_hair_03_bk_r {parent=dyn_hair_02_bk_r, loc=<6.4619 0 0>, rot=<0 0 0 1>}
 dyn_hair_bk_c_base {parent=head, loc=<2.4334 -7.4386 -0.0914>, rot=<0.1015 -0.7012 0.6984 -0.1012>}
 dyn_hair_01_bk_c {parent=dyn_hair_bk_c_base, loc=<0.2995 0 0>, rot=<0 0 0 1>}
 dyn_hair_02_bk_c {parent=dyn_hair_01_bk_c, loc=<6.4619 0 0>, rot=<0 0 0 1>}
 dyn_hair_03_bk_c {parent=dyn_hair_02_bk_c, loc=<6.4619 0 0>, rot=<0 0 0 1>}
 dyn_hair_bk_l_base {parent=head, loc=<2.9651 -7.3468 -5.3367>, rot=<0.0225 -0.7067 0.692 -0.1455>}
 dyn_hair_01_bk_l {parent=dyn_hair_bk_l_base, loc=<0.2958 0 0>, rot=<0 0 0 1>}
 dyn_hair_02_bk_l {parent=dyn_hair_01_bk_l, loc=<6.4621 0 0.0001>, rot=<0 0 0 1>}
 dyn_hair_03_bk_l {parent=dyn_hair_02_bk_l, loc=<6.4621 0 0>, rot=<0 0 0 1>}
 dyn_hair_side_r_base {parent=head, loc=<3.8873 -0.802 6.6427>, rot=<0.1463 -0.1021 0.9818 -0.0653>}
 dyn_hair_01_side_r {parent=dyn_hair_side_r_base, loc=<0.2995 0 0>, rot=<0 0 0 1>}
 dyn_hair_02_side_r {parent=dyn_hair_01_side_r, loc=<5.4391 0 0>, rot=<-0.0458 0.0198 0.0079 -0.9987>}
 dyn_hair_03_side_r {parent=dyn_hair_02_side_r, loc=<5.4391 0 0>, rot=<0.0696 0.0773 0.0159 -0.9945>}
 dyn_hair_fr_r_base {parent=head, loc=<4.8608 3.8442 7.2629>, rot=<0.1033 -0.0817 0.9912 0.0132>}
 dyn_hair_01_fr_r {parent=dyn_hair_fr_r_base, loc=<0.2995 0 0>, rot=<0 0 0 1>}
 dyn_hair_02_fr_r {parent=dyn_hair_01_fr_r, loc=<5.4391 0 0>, rot=<0 0 0.0928 -0.9957>}
 dyn_hair_03_fr_r {parent=dyn_hair_02_fr_r, loc=<5.4391 0 0>, rot=<0 0 0.0967 -0.9953>}
 clavicle_l {parent=spine_05, loc=<5.4831 -0.1053 -3.2946>, rot=<0.0465 -0.7289 -0.009 -0.6829>}
 upperarm_l {parent=clavicle_l, loc=<9.8658 -1.2136 -0.4105>, rot=<0.0164 -0.3245 -0.0269 -0.9454>}
 clavicle_r {parent=spine_05, loc=<5.4831 -0.1053 3.2946>, rot=<0.7289 0.0465 -0.6829 0.009>}
 upperarm_r {parent=clavicle_r, loc=<-9.8658 1.2136 0.4105>, rot=<0.0164 -0.3245 -0.0269 -0.9454>}
 dyn_collision_bk {parent=spine_05, loc=<11.3228 11.6416 0>, rot=<0 0 0 1>}
 dyn_collision_spine_bk {parent=spine_05, loc=<15.897 -27.1363 0>, rot=<0 0 0 1>}
 spine_05_ctrl {controlType=EULER_TRANSFORM, parent=spine_05, type=Control, loc=<0 0 0>, rot=<0 0 0 1>}
]
variables [
 Alpha_Skydive {type=float, default=0}
 Alpha_backpack {type=double, default=0}
 Alpha_cape {type=double, default=0}
 Backpack {type=bool, default=False}
 Cape {type=bool, default=False}
 Emote {type=bool, default=False}
 FE {type=bool, default=False}
 RigUnit_GetTransform_5_Transform {type=FTransform, default=(Rotation=(X=0,Y=0,Z=0,W=1),Translation=(X=0,Y=0,Z=0),Scale3D=(X=1,Y=1,Z=1))}
 RigVMFunction_MathQuaternionToEuler_Result {type=FVector, default=(X=0,Y=0,Z=0)}
 SPR_head_up {type=float, default=0}
 Skydive {type=bool, default=False}
 Upperarm_L {type=float, default=0}
 Upperarm_R {type=float, default=0}
 side_03_r {type=FTransform, default=(Rotation=(X=0,Y=0,Z=0,W=1),Translation=(X=0,Y=0,Z=0),Scale3D=(X=1,Y=1,Z=1))}
]
