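# Pre-dynamics Control Rig for the F_MED_GoldenValley face accessory, posing hair base bones from body motion.
# A sequence stores a head spherical pose reader as SPR_HeadUp, then aims back hair at spine_03 and right side hair at upperarm_r.
# An upperarm_r pose reader weights offsets on right-side strands, then front-down-right hair is constrained to neck_01 and offset by SPR_HeadUp.
# The left side aim and AlphaInterp output are unconnected.
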ControlRigBlueprint /BRCosmetics/Characters/Player/Female/Medium/Bodies/F_MED_GoldenValley/Meshes/Parts/F_MED_GoldenValley_FaceAcc_CtrlRig_PreDyn
graphs [
 RigVMFunctionLibrary {links [], nodes []}
 RigVMFunction_Sequence_3_ContainedGraph {nodes [
   n1 = RigVMFunctionEntryNode()
   n2 = &RigVMFunction_Sequence(ExecuteContext=@n1.ExecuteContext)
   n3 = &RigVMFunction_Sequence(ExecuteContext=@n2.B)
   n4 = &RigVMFunction_Sequence(ExecuteContext=@n3.B)
   n5 = &RigVMFunction_Sequence(ExecuteContext=@n4.B)
   RigVMFunctionReturnNode(B=@n3.A, A=@n2.A, C=@n4.A, D=@n5.A, E=@n5.B)
  ]}
 RigVMModel {nodes [
   RigUnit_BeginExecution = &RigUnit_BeginExecution()
   RigVMFunction_Sequence_2 = RigVMAggregateNode(ExecuteContext=@RigUnit_BeginExecution.ExecutePin)
   &RigUnit_ParentConstraint(
    AdvancedSettings=(InterpolationType=Average,RotationOrderForFilter=XZY)
    Child=(Type=Bone,Name="dyn_hair_fr_down_r_base")
    Parents=((Item=(Type=Bone,Name="neck_01"),Weight=1))
    Weight=1
    ExecutePin=@RigVMFunction_Sequence_2.D
   )
   > &RigUnit_OffsetTransformForItem(
    Item=(Type=Bone,Name="dyn_hair_fr_down_r_base")
    OffsetTransform=(Rotation=(X=0,Y=0,Z=0,W=1),Translation=(X=-1.4819,Y=-2.762,Z=0),Scale3D=(X=1,Y=1,Z=1))
    bPropagateToChildren=True
    Weight=RigVMVariableNode() {variable=SPR_HeadUp}.Value
   )
   n1 = &RigUnit_SphericalPoseReader(
    ActiveRegionScaleFactors=(PositiveWidth=1,NegativeWidth=0.5,PositiveHeight=0.75,NegativeHeight=0.75)
    ActiveRegionSize=0.05
    Debug=(bDrawDebug=false,bDraw2D=False,bDrawLocalAxes=False,DebugScale=25,DebugSegments=20,DebugThickness=0.25)
    DriverAxis=(X=0,Y=1,Z=0)
    DriverItem=(Type=Bone,Name="head")
    FalloffRegionScaleFactors=(PositiveWidth=1,NegativeWidth=0.5,PositiveHeight=1,NegativeHeight=1)
    FalloffSize=0.2
    FlipHeightScaling=False
    FlipWidthScaling=False
    OptionalParentItem=(Type=Bone,Name="None")
    RotationOffset=(X=90,Y=0,Z=-30)
    ExecutePin=@RigVMFunction_Sequence_2.A
   )
   > RigVMVariableNode(Value=@n1.OutputParam) {variable=SPR_HeadUp}
   n2 = &RigVMFunction_Sequence(ExecuteContext=@RigVMFunction_Sequence_2.B)
   /Script/ControlRig.RigUnit_AimItem(
    DebugSettings=(bEnabled=False,Scale=10,WorldOffset=(Rotation=(X=0,Y=0,Z=0,W=1),Translation=(X=0,Y=0,Z=0),Scale3D=(X=1,Y=1,Z=1)))
    Item=(Type=Bone,Name="dyn_hair_bk_l_base")
    Primary=(Weight=1,Axis=(X=1,Y=0,Z=0),Target=(X=1,Y=-7.9083,Z=-6.692),Kind=Location,Space=(Type=Bone,Name="spine_03"))
    Secondary=(Weight=1,Axis=(X=0,Y=0,Z=1),Target=(X=-0.8652,Y=7.4416,Z=10.9168),Kind=Location,Space=(Type=Bone,Name="dyn_hair_bk_c_base"))
    Weight=1
    ExecutePin=@n2.A
   ) {is=struct RigVMUnitNode}
   > &RigUnit_AimItem(
    DebugSettings=(bEnabled=False,Scale=10,WorldOffset=(Rotation=(X=0,Y=0,Z=0,W=1),Translation=(X=0,Y=0,Z=0),Scale3D=(X=1,Y=1,Z=1)))
    Item=(Type=Bone,Name="dyn_hair_bk_c_base")
    Primary=(Weight=1,Axis=(X=1,Y=0,Z=0),Target=(X=1,Y=-9.098,Z=0),Kind=Location,Space=(Type=Bone,Name="spine_03"))
    Secondary=(Weight=1,Axis=(X=0,Y=0,Z=0),Target=(X=1,Y=0,Z=0),Kind=Location,Space=(Type=Bone,Name="dyn_hair_bk_l_base"))
    Weight=1
   )
   > &RigUnit_AimItem(
    DebugSettings=(bEnabled=False,Scale=10,WorldOffset=(Rotation=(X=0,Y=0,Z=0,W=1),Translation=(X=0,Y=0,Z=0),Scale3D=(X=1,Y=1,Z=1)))
    Item=(Type=Bone,Name="dyn_hair_bk_r_base")
    Primary=(Weight=1,Axis=(X=1,Y=0,Z=0),Target=(X=1,Y=-4.9998,Z=4.0876),Kind=Location,Space=(Type=Bone,Name="spine_03"))
    Secondary=(Weight=1,Axis=(X=0,Y=0,Z=0),Target=(X=2.4594,Y=-8.5544,Z=0.0565),Kind=Location,Space=(Type=Bone,Name="dyn_hair_bk_c_base"))
    Weight=1
   )
   &RigUnit_AimItem(
    DebugSettings=(bEnabled=False,Scale=10,WorldOffset=(Rotation=(X=0,Y=0,Z=0,W=1),Translation=(X=0,Y=0,Z=0),Scale3D=(X=1,Y=1,Z=1)))
    Item=(Type=Bone,Name="dyn_hair_side_r_base")
    Primary=(Weight=1,Axis=(X=1,Y=0,Z=0),Target=(X=1,Y=0,Z=-0.0295),Kind=Location,Space=(Type=Bone,Name="upperarm_r"))
    Secondary=(Weight=1,Axis=(X=0,Y=0,Z=1),Target=(X=4.4306,Y=-0.6384,Z=10.1972),Kind=Location,Space=(Type=Bone,Name="head"))
    Weight=1
    ExecutePin=@n2.B
   )
   n3 = &RigUnit_SphericalPoseReader(
    ActiveRegionScaleFactors=(PositiveWidth=1,NegativeWidth=1,PositiveHeight=1,NegativeHeight=1)
    ActiveRegionSize=0.05
    Debug=(bDrawDebug=true,bDraw2D=False,bDrawLocalAxes=False,DebugScale=25,DebugSegments=20,DebugThickness=0.25)
    DriverAxis=(X=-1,Y=0,Z=0)
    DriverItem=(Type=Bone,Name="upperarm_r")
    FalloffRegionScaleFactors=(PositiveWidth=1,NegativeWidth=1,PositiveHeight=1,NegativeHeight=1)
    FalloffSize=0.2
    FlipHeightScaling=False
    FlipWidthScaling=False
    OptionalParentItem=(Type=Bone,Name="None")
    RotationOffset=(X=-120,Y=1,Z=27)
    ExecutePin=@RigVMFunction_Sequence_2.C
   )
   > &RigUnit_OffsetTransformForItem(
    Item=(Type=Bone,Name="dyn_hair_fr_down_r_base")
    OffsetTransform=(Rotation=(X=0,Y=0,Z=0,W=1),Translation=(X=0,Y=5,Z=-5),Scale3D=(X=1,Y=1,Z=1))
    bPropagateToChildren=True
    Weight=@n3.OutputParam
   )
   > &RigUnit_OffsetTransformForItem(
    Item=(Type=Bone,Name="dyn_hair_side_r_base")
    OffsetTransform=(Rotation=(X=0.1988,Y=0.6816,Z=-0.0616,W=-0.7015),Translation=(X=0,Y=0,Z=0),Scale3D=(X=1,Y=1,Z=1))
    bPropagateToChildren=True
    Weight=@n3.OutputParam
   )
   &RigUnit_AimItem(
    DebugSettings=(bEnabled=True,Scale=10,WorldOffset=(Rotation=(X=0,Y=0,Z=0,W=1),Translation=(X=0,Y=0,Z=0),Scale3D=(X=1,Y=1,Z=1)))
    ExecutePin=()
    Item=(Type=Bone,Name="dyn_hair_side_l_base")
    Primary=(Weight=1,Axis=(X=1,Y=0,Z=0),Target=(X=-3.6756,Y=0.0438,Z=-2.1424),Kind=Location,Space=(Type=Bone,Name="upperarm_l"))
    Secondary=(Weight=1,Axis=(X=0,Y=0,Z=-1),Target=(X=4.4306,Y=-0.6384,Z=0),Kind=Location,Space=(Type=Bone,Name="head"))
    Weight=1
   )
   AlphaInterp = &RigVMFunction_AlphaInterp(InterpSpeedDecreasing=10, InterpSpeedIncreasing=10, Scale=1, bClampResult=False, Value=@n1.OutputParam)
  ]}
]
hierarchy [
 root {loc=<0 0 0>, rot=<0 0 0 1>}
 pelvis {parent=root, loc=<0 -0.5301 98.0571>, rot=<0.0127 0.707 -0.0127 -0.707>}
 spine_01 {parent=pelvis, loc=<6.1208 -0.9328 0>, rot=<0 0 0.034 -0.9994>}
 spine_02 {parent=spine_01, loc=<6.5997 -0.2095 0>, rot=<0 0 0.0238 -0.9997>}
 spine_03 {parent=spine_02, loc=<6.4914 -0.1237 0>, rot=<0 0 0.0149 -0.9999>}
 spine_04 {parent=spine_03, loc=<6.9598 0.0794 0>, rot=<0 0 -0.0132 -0.9999>}
 spine_05 {parent=spine_04, loc=<6.9987 0 0>, rot=<0 0 -0.0015 -1>}
 neck_01 {parent=spine_05, loc=<11.3138 0.1529 0>, rot=<0 0 -0.2003 -0.9797>}
 neck_02 {parent=neck_01, loc=<4.5627 0 0>, rot=<0 0 0.0504 -0.9987>}
 head {parent=neck_02, loc=<5.3914 0 0>, rot=<0 0 0.115 -0.9934>}
 dyn_headscarf_base {parent=head, loc=<10.3043 -4.9825 -0.0893>, rot=<0.1878 -0.6831 0.6803 -0.1878>}
 dyn_headscarf {parent=dyn_headscarf_base, loc=<0.1 0 0>, rot=<0 0 0 1>}
 dyn_sun_plushie_base {parent=head, loc=<10.2216 5.2627 -4.2788>, rot=<-0.3717 -0.431 -0.1684 -0.8048>}
 dyn_sun_plushie {parent=dyn_sun_plushie_base, loc=<0.1 0 0>, rot=<0 0 0 1>}
 dyn_hair_fr_down_r_base {parent=head, loc=<-3.7514 -1.0436 7.4125>, rot=<-0.0897 -0.0362 0.971 0.2186>}
 dyn_hair_01_fr_down_r {parent=dyn_hair_fr_down_r_base, loc=<0.2753 0 0>, rot=<0 0 0 1>}
 dyn_hair_02_fr_down_r {parent=dyn_hair_01_fr_down_r, loc=<5.4391 0 0>, rot=<0.0119 0.1273 0.092 -0.9875>}
 dyn_hair_03_fr_down_r {parent=dyn_hair_02_fr_down_r, loc=<5.4391 0 0>, rot=<0 0 -0.0206 -0.9998>}
 dyn_hair_fr_l_base {parent=head, loc=<5.1137 3.6709 -6.8462>, rot=<-0.0194 -0.9996 -0.0167 0.0149>}
 dyn_hair_01_fr_l {parent=dyn_hair_fr_l_base, loc=<0.2995 0 0>, rot=<0 0 0 1>}
 dyn_hair_02_fr_l {parent=dyn_hair_01_fr_l, loc=<8.2386 0 0>, rot=<0.0042 -0.0559 -0.0757 -0.9956>}
 dyn_hair_03_fr_l {parent=dyn_hair_02_fr_l, loc=<8.2386 0 0>, rot=<0 0 -0.0575 -0.9983>}
 dyn_hair_side_l_base {parent=head, loc=<3.8999 -2.0481 -5.8012>, rot=<0.0183 -0.9864 0.0148 -0.1625>}
 dyn_hair_01_side_l {parent=dyn_hair_side_l_base, loc=<0.2995 0 0>, rot=<0 0 0 1>}
 dyn_hair_02_side_l {parent=dyn_hair_01_side_l, loc=<5.4391 0 0>, rot=<-0.0635 0.0352 -0.1198 -0.9901>}
 dyn_hair_03_side_l {parent=dyn_hair_02_side_l, loc=<5.4391 0 0>, rot=<0.2472 -0.1281 -0.0002 -0.9604>}
 dyn_hair_bk_l_base {parent=head, loc=<2.9651 -7.3468 -5.3367>, rot=<0.0225 -0.7067 0.692 -0.1455>}
 dyn_hair_01_bk_l {parent=dyn_hair_bk_l_base, loc=<0.2958 0 0>, rot=<0 0 0 1>}
 dyn_hair_02_bk_l {parent=dyn_hair_01_bk_l, loc=<6.4621 0 0.0001>, rot=<0 0 0 1>}
 dyn_hair_03_bk_l {parent=dyn_hair_02_bk_l, loc=<6.4621 0 0>, rot=<0 0 0 1>}
 dyn_hair_bk_c_base {parent=head, loc=<2.4334 -7.4386 -0.0914>, rot=<0.1015 -0.7012 0.6984 -0.1012>}
 dyn_hair_01_bk_c {parent=dyn_hair_bk_c_base, loc=<0.2995 0 0>, rot=<0 0 0 1>}
 dyn_hair_02_bk_c {parent=dyn_hair_01_bk_c, loc=<6.4619 0 0>, rot=<0 0 0 1>}
 dyn_hair_03_bk_c {parent=dyn_hair_02_bk_c, loc=<6.4619 0 0>, rot=<0 0 0 1>}
 dyn_hair_bk_r_base {parent=head, loc=<2.9686 -7.3459 5.3361>, rot=<0.1456 -0.6934 0.7054 -0.0222>}
 dyn_hair_01_bk_r {parent=dyn_hair_bk_r_base, loc=<0.2995 0 0>, rot=<0 0 0 1>}
 dyn_hair_02_bk_r {parent=dyn_hair_01_bk_r, loc=<6.4619 0 0>, rot=<0 0 0 1>}
 dyn_hair_03_bk_r {parent=dyn_hair_02_bk_r, loc=<6.4619 0 0>, rot=<0 0 0 1>}
 dyn_hair_side_r_base {parent=head, loc=<3.8873 -0.802 6.6427>, rot=<0.1463 -0.1021 0.9818 -0.0653>}
 dyn_hair_01_side_r {parent=dyn_hair_side_r_base, loc=<0.2995 0 0>, rot=<0 0 0 1>}
 dyn_hair_02_side_r {parent=dyn_hair_01_side_r, loc=<5.4391 0 0>, rot=<-0.0458 0.0198 0.0079 -0.9987>}
 dyn_hair_03_side_r {parent=dyn_hair_02_side_r, loc=<5.4391 0 0>, rot=<0.0696 0.0773 0.0159 -0.9945>}
 dyn_hair_fr_r_base {parent=head, loc=<4.8608 3.8442 7.2629>, rot=<0.1033 -0.0817 0.9912 0.0132>}
 dyn_hair_01_fr_r {parent=dyn_hair_fr_r_base, loc=<0.2995 0 0>, rot=<0 0 0 1>}
 dyn_hair_02_fr_r {parent=dyn_hair_01_fr_r, loc=<5.4391 0 0>, rot=<0 0 0.0928 -0.9957>}
 dyn_hair_03_fr_r {parent=dyn_hair_02_fr_r, loc=<5.4391 0 0>, rot=<0 0 0.0967 -0.9953>}
 clavicle_l {parent=spine_05, loc=<5.4831 -0.1053 -3.2946>, rot=<0.0465 -0.7289 -0.009 -0.6829>}
 upperarm_l {parent=clavicle_l, loc=<9.8658 -1.2136 -0.4105>, rot=<0.0164 -0.3245 -0.0269 -0.9454>}
 clavicle_r {parent=spine_05, loc=<5.4831 -0.1053 3.2946>, rot=<0.7289 0.0465 -0.6829 0.009>}
 upperarm_r {parent=clavicle_r, loc=<-9.8658 1.2136 0.4105>, rot=<0.0164 -0.3245 -0.0269 -0.9454>}
 dyn_collision_bk {parent=spine_05, loc=<11.3228 11.6416 0>, rot=<0 0 0 1>}
 dyn_collision_spine_bk {parent=spine_05, loc=<15.897 -27.1363 0>, rot=<0 0 0 1>}
]
variables [
 SPR_Clavicle_L {type=float, default=0}
 SPR_Clavicle_R {type=float, default=0}
 SPR_HeadUp {type=float, default=0}
 skydive {type=bool, default=False}
]
