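# Physics asset for the Golden Valley female medium cosmetic's face accessory mesh, driving secondary motion of hair, a headscarf and a sun plushie.
# Eight three-bone hair chains and two single-bone accessories hang from head-locked base bodies, with per-joint swing and twist limits.
# Capsules on head, neck, upper spine, shoulders and dedicated front/back collision bones form the body shape.
# Angularly free cross-constraints link some neighboring back and right-side hair bodies.
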PhysicsAsset /BRCosmetics/Characters/Player/Female/Medium/Bodies/F_MED_GoldenValley/Meshes/Parts/F_MED_GoldenValley_FaceAcc_Physics_FE
bodies {
 clavicle_l [{shapeType=Capsule, length=5, radius=5, shapeName=clavicle_l_capsule, center=<5.4214 0 0>, rotation=<rot 90 0 0>}]
 clavicle_r [{shapeType=Capsule, length=5, radius=8, shapeName=clavicle_r_capsule, center=<-6 0 2>, rotation=<rot 90 180 0>}]
 dyn_collision_bk [{shapeType=Capsule, length=8, radius=10, shapeName=dyn_collision_bk_capsule, center=<2.2122 -1.9842 0>, rotation=<rot 0 -90 0>}]
 dyn_collision_fr [
  {shapeType=Capsule, length=12, radius=8, shapeName=dyn_collision_fr_capsule, rotation=<rot 0 -90 0>}
  {shapeType=Capsule, length=12, radius=8, shapeName=dyn_collision_fr_capsule, center=<-5.9296 3.2191 0>, rotation=<rot 0 -90 0>}
 ]
 dyn_hair_01_bk_c [{shapeType=Capsule, length=4, radius=3, shapeName=dyn_hair_01_bk_c_capsule, center=<4 0 0>, rotation=<rot 90 0 0>}]
 dyn_hair_01_bk_l [{shapeType=Capsule, length=4, radius=3, shapeName=dyn_hair_01_bk_l_capsule, center=<4 0 0>, rotation=<rot 90 0 0>}]
 dyn_hair_01_bk_r [{shapeType=Capsule, length=4, radius=3, shapeName=dyn_hair_01_bk_r_capsule, center=<4 0 0>, rotation=<rot 90 0 0>}]
 dyn_hair_01_fr_down_r [{shapeType=Capsule, length=4, radius=2, shapeName=dyn_hair_01_fr_down_r_capsule, center=<4 0 0>, rotation=<rot 90 0 0>}]
 dyn_hair_01_fr_l [{shapeType=Capsule, length=4, radius=2, shapeName=dyn_hair_01_fr_l_capsule, center=<4 0 1>, rotation=<rot 90 0 0>}]
 dyn_hair_01_fr_r [{shapeType=Capsule, length=2, radius=3, shapeName=dyn_hair_01_fr_r_capsule, center=<4 0 0>, rotation=<rot 90 0 0>}]
 dyn_hair_01_side_l [{shapeType=Capsule, length=3, radius=2, shapeName=dyn_hair_01_side_l_capsule, center=<3 0 2>, rotation=<rot 90 0 0>}]
 dyn_hair_01_side_r [{shapeType=Capsule, length=3, radius=2, shapeName=dyn_hair_01_side_r_capsule, center=<3 0 0>, rotation=<rot 90 0 0>}]
 dyn_hair_02_bk_c [{shapeType=Capsule, length=2, radius=2, shapeName=dyn_hair_02_bk_c_capsule, center=<4 0 0>, rotation=<rot 90 0 90>}]
 dyn_hair_02_bk_l [{shapeType=Capsule, length=5, radius=4, shapeName=dyn_hair_02_bk_l_capsule, center=<4 0 0>, rotation=<rot 0 0 90>}]
 dyn_hair_02_bk_r [
  {shapeType=Capsule, length=4, radius=2, shapeName=dyn_hair_02_bk_r_capsule, center=<4 0 0>, rotation=<rot -0.0029 0 90>}
  {shapeType=Capsule, length=4, radius=3, shapeName=dyn_hair_02_bk_r_capsule, center=<8.9525 0 0.0002>, rotation=<rot 0 0 90>}
 ]
 dyn_hair_02_fr_down_r [{shapeType=Capsule, length=2, radius=2, shapeName=dyn_hair_02_fr_down_r_capsule, center=<2 0 0>, rotation=<rot 90 0 0>}]
 dyn_hair_02_fr_l [{shapeType=Capsule, length=4, radius=1, shapeName=dyn_hair_02_fr_l_capsule, center=<4 0 0>, rotation=<rot 90 0 0>}]
 dyn_hair_02_fr_r [{shapeType=Capsule, length=2.5, radius=2.5, shapeName=dyn_hair_02_fr_r_capsule, center=<4 0 0>, rotation=<rot 90 0 0>}]
 dyn_hair_02_side_l [{shapeType=Capsule, length=3, radius=2, shapeName=dyn_hair_02_side_l_capsule, center=<3 0 0>, rotation=<rot 90 0 0>}]
 dyn_hair_02_side_r [{shapeType=Capsule, length=3, radius=3, shapeName=dyn_hair_02_side_r_capsule, center=<3 0 0>, rotation=<rot 90 0 0>}]
 dyn_hair_03_bk_c [{shapeType=Capsule, length=8, radius=2, shapeName=dyn_hair_03_bk_c_capsule, center=<7.8204 0 0>, rotation=<rot 0 0 90>}]
 dyn_hair_03_bk_l [{shapeType=Capsule, length=9, radius=2.5, shapeName=dyn_hair_03_bk_l_capsule, center=<5.4383 0 0>, rotation=<rot 0 0 90>}]
 dyn_hair_03_bk_r [{shapeType=Capsule, length=8.5, radius=2, shapeName=dyn_hair_03_bk_r_capsule, center=<7.6268 0 0>, rotation=<rot -0.0002 0 75>}]
 dyn_hair_03_fr_down_r [{shapeType=Capsule, length=4, radius=2.5, shapeName=dyn_hair_03_fr_down_r_capsule, center=<4 0 0>, rotation=<rot 90 0 0>}]
 dyn_hair_03_fr_l [{shapeType=Capsule, length=5, radius=1.5, shapeName=dyn_hair_03_fr_l_capsule, center=<5 0 0>, rotation=<rot 90 0 0>}]
 dyn_hair_03_fr_r [{shapeType=Capsule, length=4, radius=2, shapeName=dyn_hair_03_fr_r_capsule, center=<4 0 0>, rotation=<rot 90 0 0>}]
 dyn_hair_03_side_l [
  {
   shapeType=Capsule
   length=10
   radius=3
   shapeName=dyn_hair_03_side_l_capsule
   center=<5.0178 3.7486 -0.2993>
   rotation=<rot 72.7372 -31.3276 -58.4824>
  }
  {
   shapeType=Capsule
   length=7
   radius=2
   shapeName=dyn_hair_03_side_l_capsule
   center=<3.3729 -5.8165 -2.2639>
   rotation=<rot 73.2458 -120.4234 122.758>
  }
 ]
 dyn_hair_03_side_r [
  {
   shapeType=Capsule
   length=4.5
   radius=2.5
   shapeName=dyn_hair_03_side_r_capsule
   center=<5.119 3.6981 -0.6082>
   rotation=<rot 26.9705 162.979 -95.0184>
  }
  {
   shapeType=Capsule
   length=8
   radius=3.5
   shapeName=dyn_hair_03_side_r_capsule
   center=<2.9603 -2.1252 -4.0019>
   rotation=<rot 15.8413 -153.4845 -65.5778>
  }
 ]
 dyn_hair_bk_c_base [{shapeType=Capsule, length=0.505, radius=0.505, shapeName=dyn_hair_bk_c_base_capsule, rotation=<rot 0 -90 0>}]
 dyn_hair_bk_l_base [{shapeType=Capsule, length=0.505, radius=0.505, shapeName=dyn_hair_bk_l_base_capsule, rotation=<rot 0 -90 0>}]
 dyn_hair_bk_r_base [{shapeType=Capsule, length=0.505, radius=0.505, shapeName=dyn_hair_bk_r_base_capsule, rotation=<rot 0 -90 0>}]
 dyn_hair_fr_down_r_base [{shapeType=Capsule, length=0.505, radius=0.505, shapeName=dyn_hair_fr_down_r_base_capsule, rotation=<rot 0 -90 0>}]
 dyn_hair_fr_l_base [{shapeType=Capsule, length=0.505, radius=0.505, shapeName=dyn_hair_fr_l_base_capsule, rotation=<rot 0 -90 0>}]
 dyn_hair_fr_r_base [{shapeType=Capsule, length=0.505, radius=0.505, shapeName=dyn_hair_fr_r_base_capsule, rotation=<rot 0 -90 0>}]
 dyn_hair_side_l_base [{shapeType=Capsule, length=0.505, radius=0.505, shapeName=dyn_hair_side_l_base_capsule, rotation=<rot 0 -90 0>}]
 dyn_hair_side_r_base [{shapeType=Capsule, length=0.505, radius=0.505, shapeName=dyn_hair_side_r_base_capsule, rotation=<rot 0 -90 0>}]
 dyn_headscarf [{shapeType=Capsule, length=2, radius=2.5, shapeName=dyn_headscarf_capsule, center=<4 0 0>, rotation=<rot 90 0 0>}]
 dyn_headscarf_base [{shapeType=Capsule, length=0.505, radius=0.505, shapeName=dyn_headscarf_base_capsule, rotation=<rot 0 -90 0>}]
 dyn_sun_plushie [{shapeType=Capsule, length=0, radius=3, shapeName=dyn_sun_plushie_capsule, center=<4 0 0>, rotation=<rot 90 0 0>}]
 dyn_sun_plushie_base [{shapeType=Capsule, length=0.505, radius=0.505, shapeName=dyn_sun_plushie_base_capsule, rotation=<rot 0 -90 0>}]
 head [
  {shapeType=Capsule, length=8.705, radius=5, rotation=<rot -90 -0.0012 0>}
  {shapeType=Capsule, length=4.705, radius=5, shapeName=head_capsule, center=<-0.1025 4.5907 0>, rotation=<rot -90 -0.0012 0>}
 ]
 neck_01 [{shapeType=Capsule, length=3, radius=5, shapeName=neck_01_capsule, rotation=<rot 90 0 0>}]
 neck_02 [{shapeType=Capsule, length=3, radius=4, shapeName=neck_02_capsule, rotation=<rot 90 0 0>}]
 spine_04 [{shapeType=Capsule, length=10, radius=9, shapeName=spine_04_capsule}]
 spine_05 [{shapeType=Capsule, length=10, radius=10, shapeName=spine_05_capsule, center=<0 0.2807 0>}]
 upperarm_l [{shapeType=Capsule, length=16, radius=4, shapeName=upperarm_l_capsule, center=<13.9824 0 0>, rotation=<rot 90 0 0>}]
 upperarm_r [{shapeType=Capsule, length=25, radius=7, shapeName=upperarm_r_capsule, center=<-13.9824 0 0>, rotation=<rot 90 180 0>}]
}
constraints [
 {
  bone1Name=head
  bone2Name=neck_02
  swing1LimitDegrees=45
  swing1Motion=Locked
  swing2LimitDegrees=45
  swing2Motion=Locked
  twistLimitDegrees=45
  twistMotion=Locked
 }
 {
  bone1Name=spine_05
  bone2Name=spine_04
  swing1LimitDegrees=45
  swing1Motion=Locked
  swing2LimitDegrees=45
  swing2Motion=Locked
  twistLimitDegrees=45
  twistMotion=Locked
 }
 {
  bone1Name=clavicle_l
  bone2Name=spine_05
  swing1LimitDegrees=45
  swing1Motion=Locked
  swing2LimitDegrees=45
  swing2Motion=Locked
  twistLimitDegrees=45
  twistMotion=Locked
 }
 {
  bone1Name=upperarm_l
  bone2Name=clavicle_l
  swing1LimitDegrees=45
  swing1Motion=Locked
  swing2LimitDegrees=45
  swing2Motion=Locked
  twistLimitDegrees=45
  twistMotion=Locked
 }
 {
  bone1Name=clavicle_r
  bone2Name=spine_05
  swing1LimitDegrees=45
  swing1Motion=Locked
  swing2LimitDegrees=45
  swing2Motion=Locked
  twistLimitDegrees=45
  twistMotion=Locked
 }
 {
  bone1Name=upperarm_r
  bone2Name=clavicle_r
  swing1LimitDegrees=45
  swing1Motion=Locked
  swing2LimitDegrees=45
  swing2Motion=Locked
  twistLimitDegrees=45
  twistMotion=Locked
 }
 {
  bone1Name=neck_01
  bone2Name=spine_05
  swing1LimitDegrees=45
  swing1Motion=Locked
  swing2LimitDegrees=45
  swing2Motion=Locked
  twistLimitDegrees=45
  twistMotion=Locked
 }
 {
  bone1Name=neck_02
  bone2Name=neck_01
  swing1LimitDegrees=45
  swing1Motion=Locked
  swing2LimitDegrees=45
  swing2Motion=Locked
  twistLimitDegrees=45
  twistMotion=Locked
 }
 {
  bone1Name=dyn_hair_bk_c_base
  bone2Name=head
  swing1LimitDegrees=45
  swing1Motion=Locked
  swing2LimitDegrees=45
  swing2Motion=Locked
  twistLimitDegrees=45
  twistMotion=Locked
 }
 {bone1Name=dyn_hair_01_bk_c, bone2Name=dyn_hair_bk_c_base, swing1LimitDegrees=15, swing2LimitDegrees=40, twistLimitDegrees=5}
 {bone1Name=dyn_hair_02_bk_c, bone2Name=dyn_hair_01_bk_c, swing1LimitDegrees=15, swing2LimitDegrees=40, twistLimitDegrees=5}
 {bone1Name=dyn_hair_03_bk_c, bone2Name=dyn_hair_02_bk_c, swing1LimitDegrees=15, swing2LimitDegrees=30, twistLimitDegrees=5}
 {
  bone1Name=dyn_hair_bk_l_base
  bone2Name=head
  swing1LimitDegrees=45
  swing1Motion=Locked
  swing2LimitDegrees=45
  swing2Motion=Locked
  twistLimitDegrees=45
  twistMotion=Locked
 }
 {bone1Name=dyn_hair_01_bk_l, bone2Name=dyn_hair_bk_l_base, swing1LimitDegrees=15, swing2LimitDegrees=40, twistLimitDegrees=5}
 {bone1Name=dyn_hair_02_bk_l, bone2Name=dyn_hair_01_bk_l, swing1LimitDegrees=15, swing2LimitDegrees=40, twistLimitDegrees=5}
 {bone1Name=dyn_hair_03_bk_l, bone2Name=dyn_hair_02_bk_l, swing1LimitDegrees=15, swing2LimitDegrees=30, twistLimitDegrees=5}
 {
  bone1Name=dyn_hair_bk_r_base
  bone2Name=head
  swing1LimitDegrees=45
  swing1Motion=Locked
  swing2LimitDegrees=45
  swing2Motion=Locked
  twistLimitDegrees=45
  twistMotion=Locked
 }
 {bone1Name=dyn_hair_01_bk_r, bone2Name=dyn_hair_bk_r_base, swing1LimitDegrees=15, swing2LimitDegrees=40, twistLimitDegrees=5}
 {bone1Name=dyn_hair_02_bk_r, bone2Name=dyn_hair_01_bk_r, swing1LimitDegrees=15, swing2LimitDegrees=20, twistLimitDegrees=30}
 {bone1Name=dyn_hair_03_bk_r, bone2Name=dyn_hair_02_bk_r, swing1LimitDegrees=30, swing2LimitDegrees=60, twistLimitDegrees=30}
 {
  bone1Name=dyn_hair_fr_down_r_base
  bone2Name=head
  swing1LimitDegrees=45
  swing1Motion=Locked
  swing2LimitDegrees=45
  swing2Motion=Locked
  twistLimitDegrees=45
  twistMotion=Locked
 }
 {bone1Name=dyn_hair_01_fr_down_r, bone2Name=dyn_hair_fr_down_r_base, swing1LimitDegrees=60, swing2LimitDegrees=15, twistLimitDegrees=15}
 {bone1Name=dyn_hair_02_fr_down_r, bone2Name=dyn_hair_01_fr_down_r, swing1LimitDegrees=60, swing2LimitDegrees=60, twistLimitDegrees=30}
 {bone1Name=dyn_hair_03_fr_down_r, bone2Name=dyn_hair_02_fr_down_r, swing1LimitDegrees=60, swing2LimitDegrees=60, twistLimitDegrees=30}
 {
  bone1Name=dyn_hair_fr_l_base
  bone2Name=head
  swing1LimitDegrees=45
  swing1Motion=Locked
  swing2LimitDegrees=45
  swing2Motion=Locked
  twistLimitDegrees=45
  twistMotion=Locked
 }
 {bone1Name=dyn_hair_01_fr_l, bone2Name=dyn_hair_fr_l_base, swing1LimitDegrees=10, swing2LimitDegrees=10, twistLimitDegrees=5}
 {bone1Name=dyn_hair_02_fr_l, bone2Name=dyn_hair_01_fr_l, swing1LimitDegrees=60, swing2LimitDegrees=50, twistLimitDegrees=15}
 {bone1Name=dyn_hair_03_fr_l, bone2Name=dyn_hair_02_fr_l, swing1LimitDegrees=60, swing2LimitDegrees=50, twistLimitDegrees=15}
 {
  bone1Name=dyn_hair_fr_r_base
  bone2Name=head
  swing1LimitDegrees=45
  swing1Motion=Locked
  swing2LimitDegrees=45
  swing2Motion=Locked
  twistLimitDegrees=45
  twistMotion=Locked
 }
 {bone1Name=dyn_hair_01_fr_r, bone2Name=dyn_hair_fr_r_base, swing1LimitDegrees=30, swing2LimitDegrees=20, twistLimitDegrees=10}
 {bone1Name=dyn_hair_02_fr_r, bone2Name=dyn_hair_01_fr_r, swing1LimitDegrees=60, swing2LimitDegrees=40, twistLimitDegrees=5}
 {bone1Name=dyn_hair_03_fr_r, bone2Name=dyn_hair_02_fr_r, swing1LimitDegrees=30, swing2LimitDegrees=30, twistLimitDegrees=5}
 {
  bone1Name=dyn_hair_side_l_base
  bone2Name=head
  swing1LimitDegrees=45
  swing1Motion=Locked
  swing2LimitDegrees=45
  swing2Motion=Locked
  twistLimitDegrees=45
  twistMotion=Locked
 }
 {bone1Name=dyn_hair_01_side_l, bone2Name=dyn_hair_side_l_base, swing1LimitDegrees=15, swing2LimitDegrees=5, twistLimitDegrees=5}
 {bone1Name=dyn_hair_02_side_l, bone2Name=dyn_hair_01_side_l, swing1LimitDegrees=15, swing2LimitDegrees=30, twistLimitDegrees=5}
 {bone1Name=dyn_hair_03_side_l, bone2Name=dyn_hair_02_side_l, swing1LimitDegrees=40, swing2LimitDegrees=30, twistLimitDegrees=15}
 {
  bone1Name=dyn_hair_side_r_base
  bone2Name=head
  swing1LimitDegrees=45
  swing1Motion=Locked
  swing2LimitDegrees=45
  swing2Motion=Locked
  twistLimitDegrees=45
  twistMotion=Locked
 }
 {bone1Name=dyn_hair_01_side_r, bone2Name=dyn_hair_side_r_base, swing1LimitDegrees=30, swing2LimitDegrees=30, twistLimitDegrees=5}
 {bone1Name=dyn_hair_02_side_r, bone2Name=dyn_hair_01_side_r, swing1LimitDegrees=20, swing2LimitDegrees=30, twistLimitDegrees=5}
 {bone1Name=dyn_hair_03_side_r, bone2Name=dyn_hair_02_side_r, swing1LimitDegrees=30, swing2LimitDegrees=30, twistLimitDegrees=5}
 {
  bone1Name=dyn_headscarf_base
  bone2Name=head
  swing1LimitDegrees=45
  swing1Motion=Locked
  swing2LimitDegrees=45
  swing2Motion=Locked
  twistLimitDegrees=45
  twistMotion=Locked
 }
 {bone1Name=dyn_headscarf, bone2Name=dyn_headscarf_base, swing1LimitDegrees=1, swing2LimitDegrees=30, twistLimitDegrees=1}
 {
  bone1Name=dyn_sun_plushie_base
  bone2Name=head
  swing1LimitDegrees=45
  swing1Motion=Locked
  swing2LimitDegrees=45
  swing2Motion=Locked
  twistLimitDegrees=45
  twistMotion=Locked
 }
 {bone1Name=dyn_sun_plushie, bone2Name=dyn_sun_plushie_base, swing1LimitDegrees=15, swing2LimitDegrees=30, twistLimitDegrees=5}
 {
  bone1Name=dyn_hair_02_bk_l
  bone2Name=dyn_hair_02_bk_c
  swing1LimitDegrees=45
  swing1Motion=Free
  swing2LimitDegrees=45
  swing2Motion=Free
  twistLimitDegrees=45
  twistMotion=Free
 }
 {
  bone1Name=dyn_hair_02_bk_c
  bone2Name=dyn_hair_02_bk_r
  swing1LimitDegrees=45
  swing1Motion=Free
  swing2LimitDegrees=45
  swing2Motion=Free
  twistLimitDegrees=45
  twistMotion=Free
 }
 {
  bone1Name=dyn_hair_02_side_r
  bone2Name=dyn_hair_02_bk_r
  swing1LimitDegrees=45
  swing1Motion=Free
  swing2LimitDegrees=45
  swing2Motion=Free
  twistLimitDegrees=45
  twistMotion=Free
 }
 {
  bone1Name=dyn_hair_03_side_r
  bone2Name=dyn_hair_02_bk_r
  swing1LimitDegrees=45
  swing1Motion=Free
  swing2LimitDegrees=45
  swing2Motion=Free
  twistLimitDegrees=45
  twistMotion=Free
 }
 {
  bone1Name=dyn_collision_bk
  bone2Name=spine_05
  swing1LimitDegrees=45
  swing1Motion=Free
  swing2LimitDegrees=45
  swing2Motion=Free
  twistLimitDegrees=45
  twistMotion=Free
 }
 {
  bone1Name=dyn_collision_fr
  bone2Name=spine_05
  swing1LimitDegrees=45
  swing1Motion=Free
  swing2LimitDegrees=45
  swing2Motion=Free
  twistLimitDegrees=45
  twistMotion=Free
 }
 {bone1Name=dyn_hair_01_bk_l, bone2Name=head, swing1LimitDegrees=45, swing2LimitDegrees=45, twistLimitDegrees=45}
]
