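# Post-process-style AnimBP for the GoldenValley female face accessory, adding simulated secondary motion on top of the copied character pose.
# Parent and body-mesh poses blend per bone, then pass through a pre-dynamics Control Rig.
# Three neck-relative RigidBody simulations (default, skydive, front-end) are selected by bool blends before a post-dynamics Control Rig.
# The event graph derives dynamics states, backpack/cape flags and movement-based gravity from character state.
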AnimBlueprint /BRCosmetics/Characters/Player/Female/Medium/Bodies/F_MED_GoldenValley/Meshes/Parts/F_MED_GoldenValley_FaceAcc_AnimBP
target_skeleton=&/F_MED_GoldenValley_FaceAcc_Skeleton
graphs [
 AnimGraph {nodes [
   n1 = Control Rig!(Source=Use cached pose 'PreSkydive'(), Alpha, Skydive:bool=false, Backpack:bool=false, Cape:bool=false, FE:bool=false, Emote:bool=false) -> (Pose) {controlRigAssetReference {blueprintRigClass=&/F_MED_GoldenValley_FaceAcc_CtrlRig_PostDyn.F_MED_GoldenValley_FaceAcc_CtrlRig_PostDyn_C}}
   Output Pose(Result=@n1) {layerGroup=None}
   n2 = Layered blend per bone(
    BasePose=Copy Pose From Mesh() {bUseAttachedParent=true}
    BlendPoses_0=Copy Pose From Mesh(SourceMeshComponent=Get BodyMesh() -> (BodyMesh:Skeletal Mesh Component Object Reference)) {bUseAttachedParent=true}
   ) {layerSetup [{branchFilters [{boneName=upperarm_l}, {boneName=upperarm_r}, {boneName=clavicle_l}, {boneName=clavicle_r}]}]}
   n3 = Control Rig!(Source=@n2, Alpha, skydive:bool=false) -> (Pose) {controlRigAssetReference {blueprintRigClass=&/F_MED_GoldenValley_FaceAcc_CtrlRig_PreDyn.F_MED_GoldenValley_FaceAcc_CtrlRig_PreDyn_C}}
   n4 = "Transform (Modify) Bone - Bone: dyn_collision_spine_bk"!(ComponentPose=Local To Component(LocalPose=@n3), bAlphaBoolEnabled:bool=True, Translation="0,-100,0", Rotation="0,0,0", Scale="1,1,1") -> (Pose) {
    alphaInputType=Bool
    rotationMode=BMM_Additive
    translationMode=BMM_Additive
    alphaBoolBlend {blendInTime=0.8, blendOutTime=0.8}
    boneToModify {boneName=dyn_collision_spine_bk}
    translation {y=-100}
   }
   Save cached pose 'PreDyn'(Pose=Component To Local(ComponentPose=@n4)) {cacheName=PreDyn}
   n5 = RigidBody!(
    ComponentPose=Local To Component(LocalPose=Use cached pose 'PreDyn'())
    bAlphaBoolEnabled:bool
    OverrideWorldGravity:Vector="0,0,0"
    ComponentLinearAccScale:Vector="0,0,0"
    ComponentLinearVelScale:Vector="0,0,0"
    bTransferBoneVelocities:bool=False
   ) -> (Pose) {
    alphaInputType=Bool
    bClampLinearTranslationLimitToRefPose=true
    bFreezeIncomingPoseOnStart=true
    bOverrideWorldGravity=true
    bTransferBoneVelocities=true
    overridePhysicsAsset=&/F_MED_GoldenValley_FaceAcc_Physics_Skydive
    simulationSpace=BaseBoneSpace
    alphaBoolBlend {blendOption=Linear}
    baseBoneRef {boneName=neck_01}
    componentAppliedLinearAccClamp=<100 100 100>
    componentLinearAccScale=<0.6 0.2 0.2>
    componentLinearVelScale=<0.6 0.2 0.2>
    simSpaceSettings {
     dampingAlpha=0
     maxAngularAcceleration=50
     maxAngularVelocity=8
     maxLinearAcceleration=700
     maxLinearVelocity=500
     worldAlpha=1
     externalLinearDragV=<1 1 1>
    }
   }
   n6 = Component To Local(ComponentPose=@n5)
   n7 = RigidBody(
    ComponentPose=Local To Component(LocalPose=Use cached pose 'PreDyn'())
    OverrideWorldGravity="0,0,0"
    ComponentLinearAccScale="0,0,0"
    ComponentLinearVelScale="0,0,0"
    bTransferBoneVelocities=False
   ) {
    alphaInputType=Bool
    bClampLinearTranslationLimitToRefPose=true
    bFreezeIncomingPoseOnStart=true
    bOverrideWorldGravity=true
    bTransferBoneVelocities=true
    overridePhysicsAsset=&/F_MED_GoldenValley_FaceAcc_Physics
    simulationSpace=BaseBoneSpace
    alphaBoolBlend {blendOption=Linear}
    baseBoneRef {boneName=neck_01}
    componentAppliedLinearAccClamp=<100 100 100>
    componentLinearAccScale=<0.6 0.2 0.2>
    componentLinearVelScale=<0.6 0.2 0.2>
    simSpaceSettings {
     dampingAlpha=0
     maxAngularAcceleration=50
     maxAngularVelocity=8
     maxLinearAcceleration=700
     maxLinearVelocity=2000
     worldAlpha=1
     externalLinearDragV=<1 1 1>
    }
   }
   n8 = Component To Local(ComponentPose=@n7)
   n9 = RigidBody(
    ComponentPose=Local To Component(LocalPose=Use cached pose 'PreDyn'())
    OverrideWorldGravity="0,0,0"
    ComponentLinearAccScale="0,0,0"
    ComponentLinearVelScale="0,0,0"
    bTransferBoneVelocities=False
   ) {
    alphaInputType=Bool
    bClampLinearTranslationLimitToRefPose=true
    bFreezeIncomingPoseOnStart=true
    bOverrideWorldGravity=true
    bTransferBoneVelocities=true
    overridePhysicsAsset=&/F_MED_GoldenValley_FaceAcc_Physics_FE
    simulationSpace=BaseBoneSpace
    alphaBoolBlend {blendOption=Linear}
    baseBoneRef {boneName=neck_01}
    componentAppliedLinearAccClamp=<100 100 100>
    componentLinearAccScale=<0.6 0.2 0.2>
    componentLinearVelScale=<0.6 0.2 0.2>
    simSpaceSettings {
     dampingAlpha=0
     maxAngularAcceleration=50
     maxAngularVelocity=8
     maxLinearAcceleration=700
     maxLinearVelocity=2000
     worldAlpha=1
     externalLinearDragV=<1 1 1>
    }
   }
   Save cached pose 'PreSkydive'(Pose=Blend Poses by bool(BlendPose_0=Component To Local(ComponentPose=@n9), BlendPose_1=Blend Poses by bool(BlendPose_0=@n6, BlendPose_1=@n8))) {cacheName=PreSkydive}
  ]}
 EventGraph {nodes [
   n1 = Event OnCharacterCustomizationLoaded()
   > Set BodyMesh(BodyMesh="Get (a copy)"(Array:Array of Skeletal Mesh Component Budgeted Object References=@n1.CustomMeshParts, Dimension 1=1) -> (Output:Skeletal Mesh Component Budgeted Object Reference))
   Event BlueprintUpdateAnimation()
   > n2 = Sequence!(execute) -> (then_0, then_1, then_2:Exec)
   n3 = Branch(execute=@n2.then_1, Condition=Get bIsInFrontEnd())
   > n4 = Branch(Condition=Get bIsPlayingEmote())
   > Set Dyn_FE(Dyn_FE:bool=false) -> (Output_Get:bool)
   > Set Dyn_Emote(Dyn_Emote:bool=true) -> (Output_Get:bool)
   > Set Dyn_SkyDive(Dyn_SkyDive:bool=false) -> (Output_Get:bool)
   Set Dyn_FE(execute=@n4.else, Dyn_FE=true)
   > Set Dyn_Emote(Dyn_Emote=false)
   > Set Dyn_SkyDive(Dyn_SkyDive=false)
   n5 = Branch(execute=@n3.else, Condition=Get bIsSkyDiving())
   > Set Dyn_FE(Dyn_FE=false)
   > Set Dyn_Emote(Dyn_Emote=false)
   > Set Dyn_SkyDive(Dyn_SkyDive=true)
   Set Dyn_FE(execute=@n5.else, Dyn_FE=false)
   > Set Dyn_Emote(Dyn_Emote=true)
   > Set Dyn_SkyDive(Dyn_SkyDive=false)
   n6 = CheckForBackbling()
   Set isWearingBackpack(execute=@n2.then_2, isWearingBackpack:bool=AND Boolean(A=@n6, B=NOT Boolean(A=@n6.bIsBackblingLong))) -> (Output_Get:bool)
   > Set isWearingCape(isWearingCape=AND Boolean(A=@n6, B=@n6.bIsBackblingLong))
   n7 = "Gravity Override State Function (Parameter Asset) (Nativized)"(
    execute=@n2.then_0
    CharacterPartAnimInstance=Self-Reference()
    GameTimeInSeconds=GetGameTimeInSeconds()
    DebugJointName=spine_05
    Parameters=&/F_MED_GoldenValley_FaceAcc_Parameters
   )
   n8 = OR Boolean(A=Get bIsJumping(), B=Get bIsFalling())
   > Set Jump(Jump:bool=OR Boolean(A=@n8, B=Get bIsMoving2D())) -> (Output_Get:bool)
   > n9 = Branch(Condition=@n8)
   > n10 = Branch(Condition=Get bIsMoving2D())
   > n11 = Branch(Condition=Get bIsTacticalSprinting())
   > Set GravityOverride(GravityOverride="0,0,-2000")
   Set GravityOverride(execute=@n11.else, GravityOverride="0,0,-800")
   Set GravityOverride(execute=@n10.else, GravityOverride="0,0,-400")
   Set GravityOverride(execute=@n9.else, GravityOverride=@n7)
   Set GravityOverride(GravityOverride="0,0,-4000")
  ]}
]
variables [
 Backpack {type=boolean}
 BodyMesh {type=SkeletalMeshComponent}
 Dyn_Emote {type=boolean}
 Dyn_FE {type=boolean}
 Dyn_SkyDive {type=boolean}
 Gravity Override Params {type=GravityOverrideParamsStruct}
 GravityOverride {type=Vector}
 isWearingBackpack {type=boolean}
 isWearingCape {type=boolean}
 Jump {type=boolean}
]
